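# Material function for wheel cosmetics that supplies tinted, optionally desaturated environment reflections through LocalSpace and WorldSpace outputs.
# LocalSpace samples a texture directly with the world reflection vector.
# WorldSpace parallax-corrects by intersecting a transformed reflection ray with a world-origin box sized by ParallaxSize, then samples the same texture with the hit point's XY.
# Both outputs share the Desaturate, Tint and TintAmount inputs.
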MaterialFunction /VehicleCosmetics/Wheels/Mild/Materials/MF/MF_ParallexCorrectWorldENV
graph {nodes [
  n1 = TextureObject()
  n2 = FunctionInput(Preview=Constant()) {bUsePreviewValueAsDefault=true, inputName=Desaturate, inputType=FunctionInput_Scalar}
  n3 = Desaturation(None=TextureSample(UVs=ReflectionVectorWS(), Tex=@n1).RGBA, Fraction=@n2)
  n4 = FunctionInput(Preview=Constant3Vector() {constant=<rgba 1 1 1 0>}) {bUsePreviewValueAsDefault=true, inputName=Tint}
  n5 = FunctionInput(Preview=Constant() {r=1}) {bUsePreviewValueAsDefault=true, inputName=TintAmount, inputType=FunctionInput_Scalar}
  FunctionOutput(None=LinearInterpolate(A=@n3, B=Multiply(A=@n3, B=@n4), Alpha=@n5)) {outputName=LocalSpace}
  n6 = FunctionInput(Preview=Constant3Vector() {constant=<rgba 5000 7500 5000 0>}) {bUsePreviewValueAsDefault=true, inputName=ParallaxSize}
  n7 = WorldPosition()
  n8 = Transform(None=ReflectionVectorWS())
  n9 = Subtract(A=Constant(), B=@n6)
  n10 = Max(A=Divide(A=Subtract(A=@n6, B=@n7.XYZ), B=@n8), B=Divide(A=Subtract(A=@n9, B=@n7.XYZ), B=@n8))
  n11 = Min(A=ComponentMask(None=@n10) {r=true}, B=ComponentMask(None=@n10) {g=true})
  n12 = ComponentMask(None=@n10) {b=true}
  n13 = Add(A=WorldPosition().XYZ, B=Multiply(A=Min(A=@n11, B=@n12), B=@n8))
  n14 = Desaturation(None=TextureSample(UVs=ComponentMask(None=@n13) {g=true, r=true}, Tex=@n1).RGB, Fraction=@n2)
  FunctionOutput(None=LinearInterpolate(A=@n14, B=Multiply(A=@n14, B=@n4), Alpha=@n5)) {outputName=WorldSpace}
 ]}
