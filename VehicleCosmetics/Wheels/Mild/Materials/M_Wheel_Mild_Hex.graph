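# Animated hex-detail material for the Mild cosmetic vehicle wheel, with a switchable custom-color emissive glow.
# Three rotated UV projections plus time form a hexagonal pulse mask, alongside Skin_SubUV flipbooks of pinpoint and hex textures, some offset by bump-offset parallax.
# A parallax-corrected environment function feeds diffuse and specular, and normals blend between static hexes and the flipbook.
# Material attributes pass through vehicle damage, then de-rez transition functions.
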Material /VehicleCosmetics/Wheels/Mild/Materials/M_Wheel_Mild_Hex
graph {
 properties {bAutomaticallySetUsageInEditor=false, bUseMaterialAttributes=true, bUsedWithSkeletalMesh=true, usedShadingModels=Default Lit}
 nodes [
  n1 = NamedRerouteUsage()
  ConstantBiasScale(None=@n1) {bias=0, scale=4}
  Constant() {r=1}
  Constant()
  n2 = SetMaterialAttributes!(
   MaterialAttributes
   Base Color=NamedRerouteUsage()
   Metallic=ScalarParameter() {defaultValue=0.8, parameterName=Metallic}
   Specular=NamedRerouteUsage()
   Roughness=ScalarParameter() {defaultValue=0.2, parameterName=Roughness}
   Emissive Color=NamedRerouteUsage()
   Normal=NamedRerouteUsage()
  ) {attributeSetTypes [
    #1
    #2
    #3
    #4
    #5
    #6
   ]}
  MaterialFunctionCall(In=MaterialFunctionCall(Material=@n2) {materialFunction=&MF_BaseVehicle_Damage}.Result) {materialFunction=&MF_VehicleDeRez_Transition}
  n3 = NamedRerouteUsage()
  n4 = Multiply(A=TextureCoordinate() {coordinateIndex=1}, B=ScalarParameter() {defaultValue=2, parameterName=FlipBook_Texture_02_Tile})
  n5 = AppendVector(
   A=ScalarParameter() {defaultValue=4, parameterName=FlipBook_Rows}
   B=ScalarParameter() {defaultValue=4, group=Hex_SE, parameterName=FlipBook_Columns, sortPriority=1}
  )
  n6 = StaticBoolParameter() {defaultValue=true, parameterName=Linear_Blend}
  n7 = MaterialFunctionCall!(
   Texture=TextureObject() {samplerType=SAMPLERTYPE_LinearColor, texture=/VehicleCosmetics/Wheels/Mild/Textures/Shader_Textures/T_Detail_PinPoint_Flip_N}
   SubImages=@n5
   Speed=@n3
   UVs=@n4
   Offset_Frame_Position
   Linear_Blend=@n6
  ) {materialFunction=/VehicleCosmetics/Wheels/Mild/New/Materials/Skin_SubUV}
  Multiply(A=@n7.Output_RGB)
  n8 = NamedRerouteUsage()
  Multiply(A=MaterialFunctionCall(In=@n8, Contrast=Constant() {r=10}) {materialFunction=&CheapContrast}.Result)
  n9 = BumpOffset(Coordinate=@n4, Height=@n7.Output_Alpha)
  n10 = MaterialFunctionCall(
   Texture=TextureObject() {samplerType=SAMPLERTYPE_Normal, texture=/VehicleCosmetics/Wheels/Mild/Textures/Shader_Textures/T_Detail_PinPoint_Flip_Normal}
   SubImages=@n5
   Speed=@n8
   UVs=@n9
   Linear_Blend=@n6
  ) {materialFunction=&Skin_SubUV}
  Multiply(A=@n10.Output_RGB)
  n11 = MaterialFunctionCall(
   Texture=TextureObject() {samplerType=SAMPLERTYPE_LinearColor, texture=/VehicleCosmetics/Wheels/Mild/Textures/Shader_Textures/T_Hex_8_RGB}
   SubImages=@n5
   Speed=Constant() {r=1000}
   UVs=@n4
   Linear_Blend=@n6
  ) {materialFunction=&Skin_SubUV}
  Multiply(A=@n11.Output_RGB)
  n12 = Constant() {r=2}
  n13 = LinearInterpolate(
   A=Constant() {r=0.1}
   B=MaterialFunctionCall!(Desaturate, Tint=@n12, TintAmount=@n12, ParallaxSize) {materialFunction=/VehicleCosmetics/Wheels/Mild/Materials/MF/MF_ParallexCorrectWorldENV}.WorldSpace
   Alpha=Constant() {r=0.25}
  )
  n14 = Multiply(
   A=@n13
   B=TextureSample(UVs=@n4) {samplerType=SAMPLERTYPE_LinearColor, texture=/VehicleCosmetics/Wheels/Mild/Textures/Shader_Textures/T_Detail_Hexes_N_Alpha}.A
  )
  n15 = Add(A=Multiply(A=NamedRerouteUsage(), B=Constant() {r=0.25}), B=NamedRerouteUsage())
  n16 = ComponentMask(None=MaterialFunctionCall(
    Texture=TextureObject() {samplerType=SAMPLERTYPE_LinearColor, texture=/VehicleCosmetics/Wheels/Mild/Textures/Shader_Textures/T_Detail_PinPoint_Flip_RGB}
    SubImages=@n5
    Speed=Clamp(None=@n15)
    UVs=@n9
    Linear_Blend=@n6
   ) {materialFunction=&Skin_SubUV}.Output_RGB) {g=true}
  n17 = Add(A=Constant() {r=0.2}, B=@n16)
  n18 = Multiply(A=@n17, B=@n17)
  n19 = Constant() {r=4}
  NamedRerouteDeclaration(None=LinearInterpolate(A=@n14, B=Power(Base=@n18, Exp=@n19), Alpha=@n1)) {name=Spec, nodeColor=<rgba 0.3594 0 1 1>}
  n20 = TextureCoordinate()
  n21 = ComponentMask(None=@n20) {r=true}
  n22 = ComponentMask(None=Rotator(Coordinate=@n20, Time=Constant() {r=2.093}) {speed=1}) {r=true}
  n23 = Rotator(Coordinate=@n20, Time=Constant() {r=4.189}) {speed=1}
  n24 = ConstantBiasScale(None=AppendVector(A=AppendVector(A=@n21, B=@n22), B=ComponentMask(None=@n23) {r=true})) {bias=-0.5, scale=2}
  n25 = Multiply(A=OneMinus(None=Abs(None=@n24)), B=ScalarParameter() {defaultValue=1, group=Hex_SE, parameterName=HexGlowTile, sortPriority=1})
  n26 = ComponentMask(None=@n25) {r=true}
  n27 = ComponentMask(None=@n25) {g=true}
  n28 = Add(
   A=Min(A=Min(A=@n26, B=@n27), B=ComponentMask(None=@n25) {b=true})
   B=Multiply(A=Time() {bIgnorePause=true}, B=ScalarParameter() {defaultValue=0.5, group=Hex_SE, parameterName=PippleSpeed, sortPriority=1})
  )
  n29 = Constant() {r=1}
  n30 = Add(A=Fmod(A=Multiply(A=@n28, B=@n29), B=Constant() {r=2}), B=Constant() {r=-1})
  n31 = Clamp(None=ConstantBiasScale(None=Abs(None=@n30)) {bias=-0.5, scale=2})
  n32 = Clamp(None=OneMinus(None=@n31))
  NamedRerouteDeclaration(None=Clamp(None=Multiply(A=@n32, B=@n32))) {name=ClampSpeedBot, nodeColor=<rgba 1 0.5859 0 1>}
  n33 = Clamp(None=@n31)
  NamedRerouteDeclaration(None=Clamp(None=Multiply(A=@n33, B=@n33))) {name=ClampSpeedTop, nodeColor=<rgba 0.0547 1 0 1>}
  n34 = Multiply(A=@n13, B=@n16)
  NamedRerouteDeclaration(None=MaterialFunctionCall(In=LinearInterpolate(A=@n14, B=@n34, Alpha=@n1), Contrast=ScalarParameter() {defaultValue=0.1, parameterName=Param_1}) {materialFunction=&CheapContrast}.Result) {name=Diffuse, nodeColor=<rgba 0 1 0.0391 1>}
  n35 = Constant() {r=5}
  n36 = TextureObject() {samplerType=SAMPLERTYPE_Normal, texture=&T_Detail_PinPoint_Flip_Normal}
  n37 = Power(
   Base=MaterialFunctionCall!(
    Normal=MaterialFunctionCall(Texture=@n36, SubImages=@n5, Speed=@n3, UVs=@n5, Linear_Blend=@n6) {materialFunction=&Skin_SubUV}.Output_RGB
    Exponent=Constant() {r=3}
   ) {materialFunction=/VehicleCosmetics/Wheels/Cola_Holographic/Materials/MF_Fresnel_Classic}.Result
   Exp=Constant() {r=6}
  )
  NamedRerouteDeclaration(None=Multiply(A=Multiply(A=@n35, B=@n37), B=NamedRerouteUsage())) {name=Fresnel, nodeColor=<rgba 0 0.5312 1 1>}
  NamedRerouteDeclaration(None=@n11.Output_Alpha) {name=HexDiffuseAlpha, nodeColor=<rgba 0.8516 1 0 1>}
  NamedRerouteDeclaration(None=@n11.Output_RGB) {name=HexDiffuseRGB, nodeColor=<rgba 1 0 0.3125 1>}
  n38 = VectorParameter() {group=Hex_SE, parameterName=HEX_CustomColor, sortPriority=1, defaultValue=<rgba 1 0.3225 0 1>}
  n39 = Multiply(A=@n38.RGB, B=NamedRerouteUsage())
  n40 = Constant() {r=0.5}
  n41 = Multiply(A=@n38.RGB, B=@n38.RGB)
  n42 = ScalarParameter() {defaultValue=2, group=Hex_SE, parameterName=HexGlow}
  n43 = NamedRerouteUsage()
  n44 = Clamp(None=Multiply(A=Constant() {r=0.5}, B=ComponentMask(None=@n43) {g=true}))
  n45 = NamedRerouteUsage()
  n46 = Multiply(
   A=LinearInterpolate(A=Multiply(A=@n40, B=@n38.RGB), B=Multiply(A=@n41, B=@n42), Alpha=NamedRerouteUsage())
   B=Multiply(A=@n42, B=Multiply(A=@n44, B=@n45))
  )
  n47 = Constant() {r=1}
  NamedRerouteDeclaration(None=StaticSwitchParameter(True=LinearInterpolate(A=@n39, B=@n46, Alpha=@n47), False=Constant()) {defaultValue=true, parameterName=HexGlow}) {name=Emissive, nodeColor=<rgba 0 1 0.5078 1>}
  NamedRerouteDeclaration(None=LinearInterpolate(
    A=TextureSample(UVs=@n4) {samplerType=SAMPLERTYPE_Normal, texture=/VehicleCosmetics/Wheels/Mild/Textures/Shader_Textures/T_Detail_Hexes_N}.RGB
    B=@n10.Output_RGB
    Alpha=@n8
   )) {name=Normal, nodeColor=<rgba 0.5474 0.2336 1 1>}
  NamedRerouteUsage()
  NamedRerouteUsage()
  NamedRerouteUsage()
  TextureObject() {samplerType=SAMPLERTYPE_LinearColor, texture=null}
 ]
}
